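# Animation Blueprint for the Ocean Breeze face accessory, posing glasses and headphones on the player character.
# Normally, it copies the parent pose, constrains the glasses to facial bones, and corrects headphones using neck and shoulder alphas, whose calculation chain has no event connected.
# When the FrontEnd_Idle_Bypass curve is positive, it instead layers an accessory idle sequence from the Glasses bone, time-synced to the body's idle montage.
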AnimBlueprint /BRCosmetics/Characters/Player/Male/Medium/Bodies/M_MED_OceanBreeze/Meshes/Parts/M_MED_OceanBreeze_FaceAcc_AnimBP
target_skeleton=/Game/Characters/Player/Male/Medium/Bodies/M_MED_OceanBreeze/Meshes/Parts/M_MED_OceanBreeze_FaceAcc_Skeleton
graphs [
 AnimGraph {nodes [
   n1 = "Constraint - Glasses (PPP)"!(
    ComponentPose=Local To Component(LocalPose=Copy Pose From Mesh() {bUseAttachedParent=true})
    Alpha=1
    ConstraintWeights_0=0.3
    ConstraintWeights_1:float=0.3
    ConstraintWeights_2:float=0.4
   ) -> (Pose) {
    alphaBoolBlend {blendOption=Linear}
    boneToModify {boneName=Glasses}
    constraintSetup [{targetBone {boneName=FACIAL_L_Nostril}}, {targetBone {boneName=FACIAL_R_Nostril}}, {targetBone {boneName=head}}]
    constraintWeights [1,1,1]
   }
   n2 = "Constraint - dyn_Headphone_l (PP)"!(ComponentPose=@n1, Alpha=1, ConstraintWeights_0=0.4, ConstraintWeights_1:float=0.6) -> (Pose) {
    alphaBoolBlend {blendOption=Linear}
    boneToModify {boneName=dyn_Headphone_l}
    constraintSetup [{targetBone {boneName=clavicle_l}}, {targetBone {boneName=spine_05}}]
    constraintWeights [1,1]
   }
   n3 = "Transform (Modify) Bone - Bone: dyn_Headphone_l"(ComponentPose=@n2, Alpha=Get Shoulder_rot_f_l() -> (Shoulder_rot_f_l:double), Translation="1.5,0,0", Rotation="-30,-15,0", Scale="1,1,1") {
    rotationMode=BMM_Additive
    rotationSpace=BCS_BoneSpace
    scaleMode=BMM_Additive
    scaleSpace=BCS_BoneSpace
    translationMode=BMM_Additive
    translationSpace=BCS_BoneSpace
    alphaBoolBlend {blendOption=Linear}
    boneToModify {boneName=dyn_Headphone_l}
    rotation {pitch=-30, yaw=-15}
    translation {x=1.5}
   }
   n4 = Get neck_rot_dn() -> (neck_rot_dn:double)
   n5 = "Constraint - dyn_Headphone_l (PP)"(ComponentPose=@n3, Alpha=@n4, ConstraintWeights_0=0.9, ConstraintWeights_1=0.1) {
    alphaBoolBlend {blendOption=Linear}
    boneToModify {boneName=dyn_Headphone_l}
    constraintSetup [{targetBone {boneName=spine_05}}, {targetBone {boneName=head}}]
    constraintWeights [1,1]
   }
   n6 = "Transform (Modify) Bone - Bone: dyn_Headphone_l"(ComponentPose=@n5, Alpha=@n4, Translation="3.5,0,-2", Rotation="-30,0,0", Scale="0.95,0.95,0.95") {
    rotationMode=BMM_Additive
    rotationSpace=BCS_BoneSpace
    scaleMode=BMM_Additive
    scaleSpace=BCS_BoneSpace
    translationMode=BMM_Additive
    translationSpace=BCS_BoneSpace
    alphaBoolBlend {blendOption=Linear}
    boneToModify {boneName=dyn_Headphone_l}
    rotation {pitch=-30}
    scale=<0.95 0.95 0.95>
    translation {x=3.5, z=-2}
   }
   n7 = "Transform (Modify) Bone - Bone: dyn_Headphone"(ComponentPose=@n6, Alpha=Get neck_rot_up() -> (neck_rot_up:double), Translation="0,0,0", Rotation="0,0,-10", Scale="1,1,1") {
    rotationMode=BMM_Additive
    rotationSpace=BCS_BoneSpace
    scaleMode=BMM_Additive
    scaleSpace=BCS_BoneSpace
    translationMode=BMM_Additive
    translationSpace=BCS_BoneSpace
    alphaBoolBlend {blendOption=Linear}
    boneToModify {boneName=dyn_Headphone}
    rotation {roll=-10}
   }
   n8 = Copy Pose From Mesh() {bUseAttachedParent=true}
   n9 = "Sequence Evaluator (by time) 'OceanBreeze_Male_Idle_FaceAcc'"!(ExplicitTime:float=Get GlassesPostion() -> (GlassesPostion:double)) -> (Pose:Pose Link Structure) {
    is=title AnimGraphNode_SequenceEvaluator
    sequence=/BRCosmetics/Animation/Game/MainPlayer/Menu/BR/CustomSkins/_s23/OceanBreezeM/OceanBreeze_Male_Idle_FaceAcc
   }
   Output Pose(Result=Blend Poses by bool(
     BlendPose_0=Layered blend per bone(BasePose=@n8, BlendPoses_0=@n9) {layerSetup [{branchFilters [{boneName=Glasses}]}]}
     BlendPose_1=Component To Local(ComponentPose=@n7)
     bActiveValue=Get PlayingFrontendIdle() -> (PlayingFrontendIdle:bool)
    ) {blendType=Linear}) {layerGroup=None}
  ]}
 EventGraph {nodes [
   Event BlueprintUpdateAnimation()
   > n1 = Sequence()
   n2 = GetAnimInstance(self=GetOwnerSkeletalMeshComponent())
   > Set PlayingFrontendIdle(PlayingFrontendIdle:bool="float > float"(A=GetCurveValue(self=@n2, CurveName=FrontEnd_Idle_Bypass), B=0) {is=title K2Node_CallFunction}) -> (Output_Get:bool)
   n3 = GetOwnerSkeletalMeshComponent()
   Set GlassesPostion(
    execute=@n1.then_1
    GlassesPostion:double=Montage_GetPosition!(
     self:Anim Instance Object Reference=GetAnimInstance(self=@n3)
     Montage:Anim Montage Object Reference=/BRCosmetics/Animation/Game/MainPlayer/Menu/BR/CustomSkins/_s23/OceanBreezeM/OceanBreeze_Male_Idle_Montage
    ) -> (ReturnValue:float)
   ) -> (Output_Get:double)
   Event BlueprintInitializeAnimation()
   > Set Skel_Mesh(Skel_Mesh=GetOwnerSkeletalMeshComponent())
   > n4 = Cast To PlayerPawn_Athena(Object=TryGetPawnOwner())
   > n5 = Set Pawn Owner(Pawn Owner:Fort Player Pawn Athena Object Reference=@n4) -> (Output_Get:Fort Player Pawn Athena Object Reference)
   > Set Shoulder_Mesh(Shoulder_Mesh:Skeletal Mesh Component Object Reference=GetSkeletalMeshForPartType(self=@n5)) -> (Output_Get:Skeletal Mesh Component Object Reference)
   > Set Head_Mesh(Head_Mesh:Skeletal Mesh Component Object Reference=GetSkeletalMeshForPartType(self=@n5, PartType=Head)) -> (Output_Get:Skeletal Mesh Component Object Reference)
   n6 = Get Shoulder_Mesh() -> (Shoulder_Mesh:Skeletal Mesh Component Object Reference)
   Set Shoulder_rot_f_r(Shoulder_rot_f_r:double=GetAlphaFromReferenceTransformDelta(SkeletalMeshComponent=@n6, BoneName=upperarm_r, BaseName=clavicle_r, InputRange=(X=0,Y=110), TwistAxis="0,0,1")) -> (Output_Get:double)
   > Set Shoulder_rot_f_l(Shoulder_rot_f_l:double=GetAlphaFromReferenceTransformDelta(SkeletalMeshComponent=@n6, BoneName=upperarm_l, BaseName=clavicle_l, InputRange=(X=0,Y=110), TwistAxis="0,0,1")) -> (Output_Get:double)
   > Set neck_rot_up(neck_rot_up:double=GetAlphaFromReferenceTransformDelta(
     SkeletalMeshComponent=Get Head_Mesh() -> (Head_Mesh:Skeletal Mesh Component Object Reference)
     BoneName=head
     BaseName=spine_05
     InputRange=(X=20,Y=-30)
     TwistAxis="0,0,1"
    )) -> (Output_Get:double)
   > Set neck_rot_dn(neck_rot_dn:double=GetAlphaFromReferenceTransformDelta(
     SkeletalMeshComponent=Get Head_Mesh()
     BoneName=head
     BaseName=spine_05
     DeltaComponentType=EulerX
     InputRange=(X=-10,Y=70)
     TwistAxis="0,0,1"
    )) -> (Output_Get:double)
   > Set neck_rot_r(neck_rot_r:double=GetAlphaFromReferenceTransformDelta(
     SkeletalMeshComponent=Get Head_Mesh()
     BoneName=head
     BaseName=spine_05
     DeltaComponentType=EulerY
     InputRange=(X=20,Y=60)
     TwistAxis="0,0,1"
    )) -> (Output_Get:double)
  ]}
]
variables [
 GlassesPostion {type=number}
 Head_Mesh {type=SkeletalMeshComponent}
 neck_rot_dn {type=number}
 neck_rot_l {type=number}
 neck_rot_r {type=number}
 neck_rot_up {type=number}
 Pawn Owner {type=FortPlayerPawnAthena}
 PlayingFrontendIdle {type=boolean}
 Shoulder_Mesh {type=SkeletalMeshComponent}
 Shoulder_rot_f_l {type=number}
 Shoulder_rot_f_r {type=number}
 Skel_Mesh {type=SkeletalMeshComponent}
]
